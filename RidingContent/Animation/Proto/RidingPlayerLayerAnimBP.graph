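# Player animation layer for riding wildlife creatures, implementing the VehicleFullBodyOverride layer with variants for Smackie and Nug.
# The update event derives speed, mount yaw from a spine socket, rider-to-mount aim yaw, species tags and a combat cooldown for ride mode.
# Ride mode plays idle poses with additive leans; otherwise layered per-bone blending and turn aim offsets separate upper and lower body.
# The root follows mount yaw, and weapon bones shrink to near zero.
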AnimBlueprint /RidingContent/Animation/Proto/RidingPlayerLayerAnimBP
target_skeleton=&Fortnite_M_Avg_Player_Skeleton
graphs [
 AnimGraph {nodes [Output Pose() {layerGroup=None}]}
 EventGraph {nodes [
   Event BlueprintUpdateAnimation()
   n1 = GetOwningActor()
   > Is Valid(InputObject=@n1)
   > n2 = Cast To FortPlayerPawn(Object=@n1)
   > n3 = Set FortPlayer(FortPlayer:Fort Player Pawn Object Reference=@n2) -> (Output_Get:Fort Player Pawn Object Reference)
   > n4 = Cast To CreatureBaseRidableComponent!(
    execute
    Object=Get DoNotUse_ActiveRidable!(self:Rider Component Object Reference=GetComponentByClass!(self:Actor Object Reference=@n3, ComponentClass:Actor Component Class Reference=/Script/RidingCodeRuntime.RiderComponent) -> (ReturnValue:Rider Component Object Reference)) -> (DoNotUse_ActiveRidable:Ridable Component Object Reference)
   ) -> (then, CastFailed, AsCreature Base Ridable Component:Creature Base Ridable Component Object Reference) {is=title K2Node_DynamicCast}
   > n5 = Cast To FortAIPawn!(execute, Object=GetOwner(self=@n4)) -> (then, CastFailed, AsFort AIPawn:Fort AIPawn Object Reference) {is=title K2Node_DynamicCast}
   > n6 = Set WildlifePawn(WildlifePawn:Fort AIPawn Object Reference=@n5) -> (Output_Get:Fort AIPawn Object Reference)
   > Is Valid(InputObject=@n6)
   > Set Velocity(Velocity:double=Vector Length(A=GetVelocity!(self:Actor Object Reference=@n6) -> (ReturnValue:Vector))) -> (Output_Get:double)
   > n7 = Sequence!(execute) -> (then_0, then_1, then_2:Exec)
   n8 = Get CapsuleComponent!(self:Character Object Reference=Get WildlifePawn() -> (WildlifePawn:Fort AIPawn Object Reference)) -> (CapsuleComponent:Capsule Collision Object Reference)
   n9 = BreakRotator(InRot=TransformRotation!(T:"Transform (by ref)"=GetSocketTransform(self=@n8, InSocketName=QuadSpine_A_Spine3_C, TransformSpace=RTS_World), Rotation:Rotator="0, 0, 0") -> (ReturnValue:Rotator))
   > Set MatchRootYaw(MatchRootYaw:Rotator=MakeRotator(Yaw=float - float(A=@n9.Yaw, B=90))) -> (Output_Get:Rotator)
   > Set Is Targeting(Is Targeting:bool=Get bIsTargeting!(self:Fort Player Pawn Object Reference=Get FortPlayer() -> (FortPlayer:Fort Player Pawn Object Reference)) -> (bIsTargeting)) -> (Output_Get:bool)
   > n10 = Sequence()
   > n11 = Branch(Condition=OR Boolean(A=Get bRecentlyFired() -> (bRecentlyFired:bool), B=Get IsInCombat() -> (IsInCombat:bool)))
   > Set Timer(Timer:double=5) -> (Output_Get:double)
   n12 = Branch(execute=@n11.else, Condition="float <= float"(A=Get Timer() -> (Timer:double)) -> (ReturnValue=~))
   Set Timer(execute=@n12.else, Timer=float - float(A=Get Timer(), B=GetWorldDeltaSeconds!() -> (ReturnValue:double)))
   Set ShouldTransitionToRiding(execute=@n10.then_1, ShouldTransitionToRiding:bool="float <= float"(A=Get Timer()) -> (ReturnValue=~)) -> (Output_Get:bool)
   n13 = Get FortPlayer()
   > Set IsInCombat(IsInCombat:bool=OR Boolean!(
     A=Get Is Targeting() -> (Is Targeting:bool)
     B=HasCurrentMontage!(self:Fort Pawn Object Reference=@n13) -> (ReturnValue:bool)
     C:bool=Get bIsTargetingConsumableThrow!(self:Fort Player Pawn Object Reference=@n13) -> (bIsTargetingConsumableThrow:bool)
    ) -> (ReturnValue)) -> (Output_Get:bool)
   n14 = Get IsInCombat()
   > Set IsRideMode(IsRideMode:bool=AND Boolean(A=Get ShouldTransitionToRiding() -> (ShouldTransitionToRiding:bool), B=NOT Boolean(A=@n14))) -> (Output_Get:bool)
   n15 = Get Velocity() -> (Velocity:double)
   Set IsSprinting(execute=@n7.then_1, IsSprinting:bool="float >= float"(A=@n15, B=1000)) -> (Output_Get:bool)
   > Set IsStandingStill(IsStandingStill:bool="float <= float"(A=@n15, B=200) -> (ReturnValue=~)) -> (Output_Get:bool)
   n16 = Get IsRideMode() -> (IsRideMode:bool)
   > n17 = Branch(Condition=OR Boolean!(A="float > float"(A=@n15), B=NOT Boolean(A=@n16)) -> (ReturnValue))
   n18 = GetBaseAimRotation!(self:Pawn Object Reference=Get FortPlayer()) -> (ReturnValue_Roll:float, ReturnValue_Pitch:float, ReturnValue_Yaw:float)
   n19 = GetBaseAimRotation(self=Get WildlifePawn())
   > Set LowerBodyYaw(LowerBodyYaw:double=NormalizeAxis!(Angle:float=float - float(A=@n18.ReturnValue_Yaw, B=@n19.ReturnValue_Yaw)) -> (ReturnValue:float)) -> (Output_Get:double)
   Set LowerBodyYaw(execute=@n17.else, LowerBodyYaw=0)
   n20 = Get WildlifePawn()
   Is Valid(exec=@n7.then_2, InputObject=@n20)
   n21 = Get GameplayTags!(self:Fort Pawn Object Reference=@n20) -> (GameplayTags:Gameplay Tag Container Structure)
   > Set IsSmackie(IsSmackie:bool=HasTag!(
     TagContainer:"Gameplay Tag Container Structure (by ref)"=@n21
     Tag:Gameplay Tag Structure=(TagName="Pawn.Athena.NPC.Wildlife.Simple.Smackie")
     bExactMatch:bool=false
    ) -> (ReturnValue:bool)) -> (Output_Get:bool)
   > Set IsNug(IsNug:bool=HasTag(TagContainer=@n21, Tag=(TagName="Pawn.Athena.NPC.Wildlife.Prey.Nug"), bExactMatch=false)) -> (Output_Get:bool)
  ]}
 VehicleFinalPoseOverride {nodes [
   Input Pose() {name=InFinalPose}
   Output Pose() {layerGroup=VehicleOverride}
  ]}
 VehicleFullBodyOverride {nodes [
   Apply Additive!(
    Base
    Additive="Sequence Player (additive) 'Proto_Riding_Fast_Pose'"() {sequence=&/Proto_Riding_Fast_Pose}
    bAlphaBoolEnabled:bool=Get IsSprinting() -> (IsSprinting:bool)
   ) -> (Pose) {alphaInputType=Bool, alphaBoolBlend {blendInTime=0.35, blendOption=Linear, blendOutTime=0.35}}
   n1 = Get LowerBodyYaw() -> (LowerBodyYaw:double)
   n2 = Get LowerBodyYaw()
   n3 = AimOffset Player 'LowerBodyTurnAO'!(X:float=@n2, BasePose:Pose Link Structure=Use cached pose 'PostMaskingPose'(), Alpha:float=1) -> (Pose:Pose Link Structure) {blendSpace=&/LowerBodyTurnAO, alphaBoolBlend {blendOption=Linear}}
   n4 = AimOffset Player 'LowerBodyTurnAO_Nug'!(X:float=@n2, BasePose:Pose Link Structure=Use cached pose 'PostMaskingPose'(), Alpha:float=1) -> (Pose:Pose Link Structure) {blendSpace=&/LowerBodyTurnAO_Nug, alphaBoolBlend {blendOption=Linear}}
   n5 = Get IsNug() -> (IsNug:bool)
   n6 = Blend Poses by bool(
    BlendPose_0=Apply Additive!(
     Base=Use cached pose 'IdlePose'()
     Additive=Blendspace Player 'Proto_Riding_Leans_BS'!(X=@n1) -> (Pose) {blendSpace=&/Proto_Riding_Leans_BS}
     Alpha:float
    ) -> (Pose) {alphaBoolBlend {blendOption=Linear}}
    BlendPose_1=Blend Poses by bool(
     BlendPose_0=Use cached pose 'PostMaskingPose'()
     BlendPose_1=Blend Poses by bool(BlendPose_0=@n4, BlendPose_1=@n3, BlendTime_0=0.4, bActiveValue=@n5) {blendType=Linear}
     BlendTime_0=0.4
     bActiveValue=Get IsSmackie() -> (IsSmackie:bool)
    ) {blendType=Linear}
    BlendTime_0=0.4
    bActiveValue=Get IsRideMode()
   ) {blendType=Linear}
   n7 = Get LowerBodyYaw()
   n8 = SelectFloat(A=float * float(A=@n7, B=-1) {is=title K2Node_PromotableOperator}, B=@n7, bPickA=Get IsSmackie())
   n9 = "Transform (Modify) Bone - Bone: Root"!(
    ComponentPose=Local To Component(LocalPose=Rotate Root Bone!(BasePose:Pose Link Structure=@n6, Pitch:float=0, Yaw:float=@n8, MeshToComponent:Rotator="0,0,0") -> (Pose:Pose Link Structure) {is=title AnimGraphNode_RotateRootBone})
    bAlphaBoolEnabled:bool=True
    Translation="0,0,0"
    Rotation=Get MatchRootYaw() -> (MatchRootYaw:Rotator)
    Scale="1,1,1"
   ) -> (Pose) {
    alphaInputType=Bool
    rotationMode=BMM_Replace
    rotationSpace=BCS_WorldSpace
    translationSpace=BCS_WorldSpace
    alphaBoolBlend {blendOption=Linear}
    boneToModify {boneName=Root}
    translation {x=100}
   }
   n10 = "Transform (Modify) Bone - Bone: weapon_r"!(ComponentPose=@n9, bAlphaBoolEnabled:bool=True, Scale="0.001,0.001,0.001") -> (Pose) {alphaInputType=Bool, scaleMode=BMM_Replace, alphaBoolBlend {blendOption=Linear}, boneToModify {boneName=weapon_r}, scale=<0.001 0.001 0.001>}
   n11 = "Transform (Modify) Bone - Bone: weapon_l"!(ComponentPose=@n10, bAlphaBoolEnabled:bool=True, Scale="0.001,0.001,0.001") -> (Pose) {alphaInputType=Bool, scaleMode=BMM_Replace, alphaBoolBlend {blendOption=Linear}, boneToModify {boneName=weapon_l}, scale=<0.001 0.001 0.001>}
   n12 = "Transform (Modify) Bone - Bone: hand_attach_r"!(ComponentPose=@n11, bAlphaBoolEnabled:bool=True, Scale="0.001,0.001,0.001") -> (Pose) {
    alphaInputType=Bool
    scaleMode=BMM_Replace
    alphaBoolBlend {blendOption=Linear}
    boneToModify {boneName=hand_attach_r}
    scale=<0.001 0.001 0.001>
   }
   n13 = "Transform (Modify) Bone - Bone: hand_attach_l"!(ComponentPose=@n12, bAlphaBoolEnabled:bool=True, Scale="0.001,0.001,0.001") -> (Pose) {
    alphaInputType=Bool
    scaleMode=BMM_Replace
    alphaBoolBlend {blendOption=Linear}
    boneToModify {boneName=hand_attach_l}
    scale=<0.001 0.001 0.001>
   }
   n14 = Component To Local(ComponentPose=@n13)
   Output Pose(Result=Slot 'UpperBody'(Source=Slot 'DefaultSlot'(Source=@n14)) {slotName=UpperBody}) {layerGroup=VehicleOverride}
   Save cached pose 'UpperAndLowerPose'(Pose=Input Pose() {name=InPoseFullBody}) {cacheName=UpperAndLowerPose}
   n15 = Blend Poses by bool(
    BlendPose_0=Sequence Player 'Proto_Riding_IdleBreak_Pose'() {sequence=&/Proto_Riding_IdleBreak_Pose}
    BlendPose_1=Sequence Player 'Riding_Idle_Pose'() {sequence=&/Riding_Idle_Pose}
    BlendTime_0=0.4
    bActiveValue=AND Boolean(A=Get IsStandingStill() -> (IsStandingStill:bool), B=Get IsRideMode())
   ) {blendType=Linear}
   n16 = Sequence Player 'Nug_Riding_Idle_Pose'() {sequence=&/Nug_Riding_Idle_Pose}
   n17 = Get IsNug()
   Save cached pose 'IdlePose'(Pose=Blend Poses by bool(
     BlendPose_0=Sequence Player 'Smackie_Riding_Idle_Pose'() {sequence=&/Smackie_Riding_Idle_Pose}
     BlendPose_1=Blend Poses by bool(BlendPose_0=@n16, BlendPose_1=@n15, BlendTime_0=0.4, bActiveValue=@n17) {blendType=Linear}
     BlendTime_0=0.4
     bActiveValue=Get IsSmackie()
    ) {blendType=Linear}) {cacheName=IdlePose}
   n18 = Get LowerBodyYaw()
   n19 = Use cached pose 'UpperAndLowerPose'()
   n20 = Layered blend per bone(
    BasePose=Blend Poses by bool(
     BlendPose_0=Blendspace Player 'Smackie_LowerBodyTurn_BS'!(X=@n18) -> (Pose) {blendSpace=&/Smackie_LowerBodyTurn_BS}
     BlendPose_1=Use cached pose 'IdlePose'()
     BlendTime_0=0.4
     bActiveValue=Get IsSmackie()
    ) {blendType=Linear}
    BlendPoses_0=Blend Poses by bool(
     BlendPose_0=Rotate Root Bone(BasePose=@n19, Pitch=0, Yaw=@n18, MeshToComponent="0,0,0")
     BlendPose_1=Use cached pose 'UpperAndLowerPose'()
     BlendTime_0=0.4
     bActiveValue=Get IsSmackie()
    ) {blendType=Linear}
   ) {
    bBlendRootMotionBasedOnRootBone=false
    bMeshSpaceRotationBlend=true
    layerSetup [{branchFilters [
       {boneName=Root}
       {blendDepth=-1, boneName=pelvis}
       {boneName=Spine_01}
       {blendDepth=-1, boneName=thigh_l}
       {blendDepth=-1, boneName=thigh_r}
       {blendDepth=-1, boneName=ik_foot_root}
      ]}]
   }
   Save cached pose 'PostMaskingPose'(Pose=@n20) {cacheName=PostMaskingPose}
   Get NegativeTurnYaw() -> (NegativeTurnYaw:float)
  ]}
 VehicleLowerBodyOverride {nodes [
   Input Pose() {name=InPoseLowerBodyDefault}
   Input Pose() {name=InPoseUpperBody}
   Output Pose() {layerGroup=VehicleOverride}
  ]}
 VehicleSplitBodyOverride {nodes [
   Input Pose() {name=InPoseSplitBody}
   Input Pose() {name=InPoseUpperAndLowerBody}
   Input Pose() {name=InPoseUpperAndLowerBodyRemoveRoot}
   Output Pose() {layerGroup=Vehicle Override}
  ]}
]
variables [
 FortPlayer {type=FortPlayerPawn}
 Is Targeting {type=boolean}
 IsInCombat {type=boolean}
 IsNug {type=boolean}
 IsRideMode {type=boolean}
 IsSmackie {type=boolean}
 IsSprinting {type=boolean}
 IsStandingStill {type=boolean}
 LowerBodyYaw {type=number}
 MatchRootYaw {type=Rotator}
 ShouldTransitionToRiding {type=boolean}
 Timer {type=number}
 Velocity {type=number}
 WildlifePawn {type=FortAIPawn, default=null}
]
